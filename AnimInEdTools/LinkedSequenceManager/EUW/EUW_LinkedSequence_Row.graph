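# Editor Utility Widget row for the Linked Sequence Manager, showing one Movie Scene binding with a checkbox, name and AnimSequence picker.
# The picker is a SinglePropertyView pointed at the widget's own AnimSequence variable in PreConstruct.
# Checkbox changes recolor the name and show or hide the picker and "+" button, which appears when no sequence is set.
# Clicking "+" broadcasts the binding through a dispatcher, leaving asset creation to listeners.
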EditorUtilityWidgetBlueprint /AnimInEdTools/LinkedSequenceManager/EUW/EUW_LinkedSequence_Row
graphs [
 Create New AnimSequence for Binding {nodes [Create New AnimSequence for Binding!() -> (then, Binding:Movie Scene Binding Proxy Structure) {is=title K2Node_FunctionEntry}]}
 EventGraph {nodes [
   "On Property Changed (AnimSequencePropertyView)"!() -> (OutputDelegate, then, PropertyName:Name=None) {is=title K2Node_ComponentBoundEvent}
   > n1 = Branch(Condition=IsValid(Object=Get AnimSequence() -> (AnimSequence:Anim Sequence Object Reference)))
   > SetVisibility(self=Get CreateNewAnimSequenceButton() -> (CreateNewAnimSequenceButton:Editor Utility Button Object Reference), InVisibility=Collapsed)
   SetVisibility(execute=@n1.else, self=Get CreateNewAnimSequenceButton(), InVisibility=Visible)
   "On Check State Changed (CheckBox)"!() -> (OutputDelegate, then, bIsChecked:bool) {is=title K2Node_ComponentBoundEvent}
   > Update Appearance()
   "On Clicked (CreateNewAnimSequenceButton)"() {is=title K2Node_ComponentBoundEvent}
   > Call Create New Anim Sequence For Binding!(
    execute:Exec
    self:EUW Linked Sequence Row Object Reference
    Binding:Movie Scene Binding Proxy Structure=Get In Binding() -> (In Binding:Movie Scene Binding Proxy Structure)
   ) -> (then:Exec) {is=title K2Node_CallDelegate}
   Update Appearance() {is=title K2Node_CustomEvent}
   > n2 = Sequence!(execute) -> (then_0, then_1, then_2:Exec)
   n3 = Get CheckBox() -> (CheckBox:Check Box Object Reference)
   n4 = Make Slate Color(SpecifiedColor=(R=0.0704,G=0.0704,B=0.0704,A=1))
   > SetColorAndOpacity(
    self=Get BindingName() -> (BindingName:Text Object Reference)
    InColorAndOpacity=Select!(
     Unchecked:Slate Color Structure=@n4
     Checked:Slate Color Structure=Make Slate Color(SpecifiedColor=(R=0.5255,G=0.5255,B=0.5255,A=1))
     Undetermined:Slate Color Structure=@n4
     Index:ECheckBoxState Enum=GetCheckedState!(self:Check Box Object Reference=@n3) -> (ReturnValue:ECheckBoxState Enum=Unchecked)
    ) -> (ReturnValue:Slate Color Structure)
   )
   n5 = Switch on ECheckBoxState!(execute=@n2.then_1, Selection:ECheckBoxState Enum=GetCheckedState(self=Get CheckBox()) -> (ReturnValue=Unchecked)) -> (Unchecked:Exec, Checked:Exec, Undetermined:Exec) {is=title K2Node_SwitchEnum}
   SetVisibility(
    execute=[@n5.Unchecked, @n5.Undetermined]
    self=Get AnimSequencePropertyView() -> (AnimSequencePropertyView:Single Property View Object Reference)
    InVisibility=Collapsed
   )
   > SetVisibility(self=Get CreateNewAnimSequenceButton(), InVisibility=Collapsed)
   SetVisibility(execute=@n5.Checked, self=Get AnimSequencePropertyView(), InVisibility=Visible)
   > n6 = Branch(Condition=IsValid(Object=Get AnimSequence()))
   > SetVisibility(self=Get CreateNewAnimSequenceButton(), InVisibility=Collapsed)
   SetVisibility(execute=@n6.else, self=Get CreateNewAnimSequenceButton(), InVisibility=Visible)
   Event PreConstruct()
   > SetObject(self=Get AnimSequencePropertyView(), NewObject=Self-Reference())
   Event Construct()
   > SetVisibility(self=Get CreateNewAnimSequenceButton(), InVisibility=Collapsed)
  ]}
]
variables [
 AnimSequence {type=AnimSequence, default=null}
 Binding {type=string, default=""}
 Create New AnimSequence for Binding {default=()}
 In Binding {type=MovieSceneBindingProxy, default {bindingId=#1, sequence=null}}
]
widgets {
 info {rootWidgetClass=&Border, widgetCount=8}
 widgets [
  Border_178 {widgetClassPath=&Border, widget=Border_178, properties {brushColor=<rgba 0.0103 0.0103 0.0103 1>, padding=<ltrb 20 5 20 1>}}
  Border {
   widgetClassPath=&Border
   parent=Border_178
   widget=Border
   properties {brushColor=<rgba 0.0176 0.0176 0.0176 1>, padding=<ltrb 0 5 0 5>}
   slot {class=BorderSlot, properties {padding=<ltrb 20 5 20 1>}}
  }
  HorizontalBox {widgetClassPath=&HorizontalBox, parent=Border, widget=HorizontalBox, slot {class=BorderSlot, properties {padding=<ltrb 0 5 0 5>}}}
  CheckBox {
   widgetClassPath=/Script/UMG.CheckBox
   bIsVariable=true
   parent=HorizontalBox
   widget=CheckBox
   properties {widgetStyle {
     checkBoxType=CheckBox
     backgroundHoveredImage {
      bIsDynamicallyLoaded=false
      drawAs=RoundedBox
      imageType=NoImage
      mirroring=NoMirror
      resourceName=None
      resourceObject=null
      tiling=NoTile
      imageSize=<18 18>
      margin=<ltrb 0 0 0 0>
      outlineSettings {
       bUseBrushTransparency=false
       roundingType=FixedRadius
       width=1
       color {colorUseRule=UseColor_ColorTable, specifiedColor=<rgba 0 0 0 0>}
       cornerRadii=<3 3 3 3>
      }
      tintColor {colorUseRule=UseColor_Specified, specifiedColor=<rgba 0.0048 0.0048 0.0048 1>}
      uVRegion {bIsValid=false, max=<0 0>, min=<0 0>}
     }
     backgroundImage {
      bIsDynamicallyLoaded=false
      drawAs=RoundedBox
      imageType=NoImage
      mirroring=NoMirror
      resourceName=None
      resourceObject=null
      tiling=NoTile
      imageSize=<18 18>
      margin=<ltrb 0 0 0 0>
      outlineSettings {
       bUseBrushTransparency=false
       roundingType=FixedRadius
       width=1
       color {colorUseRule=UseColor_ColorTable, specifiedColor=<rgba 0 0 0 0>}
       cornerRadii=<3 3 3 3>
      }
      tintColor {colorUseRule=UseColor_Specified, specifiedColor=<rgba 0.0048 0.0048 0.0048 1>}
      uVRegion {bIsValid=false, max=<0 0>, min=<0 0>}
     }
     backgroundPressedImage {
      bIsDynamicallyLoaded=false
      drawAs=RoundedBox
      imageType=NoImage
      mirroring=NoMirror
      resourceName=None
      resourceObject=null
      tiling=NoTile
      imageSize=<18 18>
      margin=<ltrb 0 0 0 0>
      outlineSettings {
       bUseBrushTransparency=false
       roundingType=FixedRadius
       width=1
       color {colorUseRule=UseColor_ColorTable, specifiedColor=<rgba 0 0 0 0>}
       cornerRadii=<3 3 3 3>
      }
      tintColor {colorUseRule=UseColor_Specified, specifiedColor=<rgba 0.0048 0.0048 0.0048 1>}
      uVRegion {bIsValid=false, max=<0 0>, min=<0 0>}
     }
     borderBackgroundColor {colorUseRule=UseColor_Specified, specifiedColor=<rgba 1 1 1 1>}
     checkedForeground {colorUseRule=UseColor_Specified, specifiedColor=<rgba 1 1 1 1>}
     checkedHoveredForeground {colorUseRule=UseColor_Specified, specifiedColor=<rgba 1 1 1 1>}
     checkedHoveredImage {
      bIsDynamicallyLoaded=false
      drawAs=Image
      imageType=Vector
      mirroring=NoMirror
      resourceName=../../../Engine/Content/Slate/Starship/CoreWidgets/CheckBox/check.svg
      resourceObject=null
      tiling=NoTile
      imageSize=<16 16>
      margin=<ltrb 0 0 0 0>
      outlineSettings {
       bUseBrushTransparency=false
       roundingType=HalfHeightRadius
       width=0
       color {colorUseRule=UseColor_Specified, specifiedColor=<rgba 0 0 0 0>}
       cornerRadii=<0 0 0 0>
      }
      tintColor {colorUseRule=UseColor_Specified, specifiedColor=<rgba 0.0044 0.2384 1 1>}
      uVRegion {bIsValid=false, max=<0 0>, min=<0 0>}
     }
     checkedImage {
      bIsDynamicallyLoaded=false
      drawAs=Image
      imageType=Vector
      mirroring=NoMirror
      resourceName=../../../Engine/Content/Slate/Starship/CoreWidgets/CheckBox/check.svg
      resourceObject=null
      tiling=NoTile
      imageSize=<16 16>
      margin=<ltrb 0 0 0 0>
      outlineSettings {
       bUseBrushTransparency=false
       roundingType=HalfHeightRadius
       width=0
       color {colorUseRule=UseColor_Specified, specifiedColor=<rgba 0 0 0 0>}
       cornerRadii=<0 0 0 0>
      }
      tintColor {colorUseRule=UseColor_Specified, specifiedColor=<rgba 0 0.162 0.7454 1>}
      uVRegion {bIsValid=false, max=<0 0>, min=<0 0>}
     }
     checkedPressedForeground {colorUseRule=UseColor_Specified, specifiedColor=<rgba 1 1 1 1>}
     checkedPressedImage {
      bIsDynamicallyLoaded=false
      drawAs=Image
      imageType=Vector
      mirroring=NoMirror
      resourceName=../../../Engine/Content/Slate/Starship/CoreWidgets/CheckBox/check.svg
      resourceObject=null
      tiling=NoTile
      imageSize=<16 16>
      margin=<ltrb 0 0 0 0>
      outlineSettings {
       bUseBrushTransparency=false
       roundingType=HalfHeightRadius
       width=0
       color {colorUseRule=UseColor_Specified, specifiedColor=<rgba 0 0 0 0>}
       cornerRadii=<0 0 0 0>
      }
      tintColor {colorUseRule=UseColor_Specified, specifiedColor=<rgba 0 0.0802 0.3515 1>}
      uVRegion {bIsValid=false, max=<0 0>, min=<0 0>}
     }
     checkedSlateSound {resourceObject=null}
     foregroundColor {colorUseRule=UseColor_Specified, specifiedColor=<rgba 1 1 1 1>}
     hoveredForeground {colorUseRule=UseColor_Specified, specifiedColor=<rgba 1 1 1 1>}
     hoveredSlateSound {resourceObject=null}
     padding=<ltrb 2 0 0 0>
     pressedForeground {colorUseRule=UseColor_Specified, specifiedColor=<rgba 1 1 1 1>}
     uncheckedHoveredImage {
      bIsDynamicallyLoaded=false
      drawAs=NoDrawType
      imageType=NoImage
      mirroring=NoMirror
      resourceName=None
      resourceObject=null
      tiling=NoTile
      imageSize=<0 0>
      margin=<ltrb 0 0 0 0>
      outlineSettings {
       bUseBrushTransparency=false
       roundingType=HalfHeightRadius
       width=0
       color {colorUseRule=UseColor_Specified, specifiedColor=<rgba 0 0 0 0>}
       cornerRadii=<0 0 0 0>
      }
      tintColor {colorUseRule=UseColor_Specified, specifiedColor=<rgba 1 1 1 1>}
      uVRegion {bIsValid=false, max=<0 0>, min=<0 0>}
     }
     uncheckedImage {
      bIsDynamicallyLoaded=false
      drawAs=NoDrawType
      imageType=NoImage
      mirroring=NoMirror
      resourceName=None
      resourceObject=null
      tiling=NoTile
      imageSize=<0 0>
      margin=<ltrb 0 0 0 0>
      outlineSettings {
       bUseBrushTransparency=false
       roundingType=HalfHeightRadius
       width=0
       color {colorUseRule=UseColor_Specified, specifiedColor=<rgba 0 0 0 0>}
       cornerRadii=<0 0 0 0>
      }
      tintColor {colorUseRule=UseColor_Specified, specifiedColor=<rgba 1 1 1 1>}
      uVRegion {bIsValid=false, max=<0 0>, min=<0 0>}
     }
     uncheckedPressedImage {
      bIsDynamicallyLoaded=false
      drawAs=NoDrawType
      imageType=NoImage
      mirroring=NoMirror
      resourceName=None
      resourceObject=null
      tiling=NoTile
      imageSize=<0 0>
      margin=<ltrb 0 0 0 0>
      outlineSettings {
       bUseBrushTransparency=false
       roundingType=HalfHeightRadius
       width=0
       color {colorUseRule=UseColor_Specified, specifiedColor=<rgba 0 0 0 0>}
       cornerRadii=<0 0 0 0>
      }
      tintColor {colorUseRule=UseColor_Specified, specifiedColor=<rgba 1 1 1 1>}
      uVRegion {bIsValid=false, max=<0 0>, min=<0 0>}
     }
     uncheckedSlateSound {resourceObject=null}
     undeterminedForeground {colorUseRule=UseColor_Specified, specifiedColor=<rgba 1 1 1 1>}
     undeterminedHoveredImage {
      bIsDynamicallyLoaded=false
      drawAs=Image
      imageType=Vector
      mirroring=NoMirror
      resourceName=../../../Engine/Content/Slate/Starship/CoreWidgets/CheckBox/indeterminate.svg
      resourceObject=null
      tiling=NoTile
      imageSize=<16 16>
      margin=<ltrb 0 0 0 0>
      outlineSettings {
       bUseBrushTransparency=false
       roundingType=HalfHeightRadius
       width=0
       color {colorUseRule=UseColor_Specified, specifiedColor=<rgba 0 0 0 0>}
       cornerRadii=<0 0 0 0>
      }
      tintColor {colorUseRule=UseColor_Specified, specifiedColor=<rgba 0.0044 0.2384 1 1>}
      uVRegion {bIsValid=false, max=<0 0>, min=<0 0>}
     }
     undeterminedImage {
      bIsDynamicallyLoaded=false
      drawAs=Image
      imageType=Vector
      mirroring=NoMirror
      resourceName=../../../Engine/Content/Slate/Starship/CoreWidgets/CheckBox/indeterminate.svg
      resourceObject=null
      tiling=NoTile
      imageSize=<16 16>
      margin=<ltrb 0 0 0 0>
      outlineSettings {
       bUseBrushTransparency=false
       roundingType=HalfHeightRadius
       width=0
       color {colorUseRule=UseColor_Specified, specifiedColor=<rgba 0 0 0 0>}
       cornerRadii=<0 0 0 0>
      }
      tintColor {colorUseRule=UseColor_Specified, specifiedColor=<rgba 0 0.162 0.7454 1>}
      uVRegion {bIsValid=false, max=<0 0>, min=<0 0>}
     }
     undeterminedPressedImage {
      bIsDynamicallyLoaded=false
      drawAs=Image
      imageType=Vector
      mirroring=NoMirror
      resourceName=../../../Engine/Content/Slate/Starship/CoreWidgets/CheckBox/indeterminate.svg
      resourceObject=null
      tiling=NoTile
      imageSize=<16 16>
      margin=<ltrb 0 0 0 0>
      outlineSettings {
       bUseBrushTransparency=false
       roundingType=HalfHeightRadius
       width=0
       color {colorUseRule=UseColor_Specified, specifiedColor=<rgba 0 0 0 0>}
       cornerRadii=<0 0 0 0>
      }
      tintColor {colorUseRule=UseColor_Specified, specifiedColor=<rgba 0 0.0802 0.3515 1>}
      uVRegion {bIsValid=false, max=<0 0>, min=<0 0>}
     }
    }}
   slot {class=HorizontalBoxSlot, properties {padding=<ltrb 25 20 0 20>}}
  }
  BindingName {
   widgetClassPath=&TextBlock
   bIsVariable=true
   parent=HorizontalBox
   widget=BindingName
   properties {
    justification=Center
    text=Binding Name
    colorAndOpacity {colorUseRule=UseColor_Specified, specifiedColor=<rgba 0.5255 0.5255 0.5255 1>}
    font {
     bForceMonospaced=false
     bMaterialIsStencil=false
     fontMaterial=null
     fontObject=&Roboto
     letterSpacing=0
     monospacedWidth=1
     size=10
     skewAmount=0
     typefaceFontName=Bold
     outlineSettings {
      bApplyOutlineToDropShadows=false
      bMiteredCorners=false
      bSeparateFillAlpha=false
      outlineMaterial=null
      outlineSize=0
      outlineColor=<rgba 0 0 0 1>
     }
    }
   }
   slot {class=HorizontalBoxSlot, properties {horizontalAlignment=HAlign_Left, verticalAlignment=VAlign_Center, padding=<ltrb 25 2 25 2>}}
  }
  AnimSequencePropertyView {
   widgetClassPath=&SinglePropertyView
   parent=HorizontalBox
   widget=AnimSequencePropertyView
   properties {propertyName=AnimSequence}
   slot {class=HorizontalBoxSlot, properties {size {sizeRule=Fill, value=1}}}
  }
  CreateNewAnimSequenceButton {
   widgetClassPath=&EditorUtilityButton
   parent=HorizontalBox
   widget=CreateNewAnimSequenceButton
   slot {class=HorizontalBoxSlot}
  }
  TextBlock_155 {
   widgetClassPath=&TextBlock
   parent=CreateNewAnimSequenceButton
   widget=TextBlock_155
   properties {text=+}
   slot {class=ButtonSlot}
  }
 ]
}
